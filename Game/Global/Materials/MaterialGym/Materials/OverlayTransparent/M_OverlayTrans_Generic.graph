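# Translucent unlit overlay material in the MaterialGym, acting as a showcase of MF_OverlayTrans_Generic and enabled for skeletal meshes and mesh deformers.
# Showcase parameters for base color, roughness and emissive join a tiling noise normal in a material attributes set that feeds the function.
# Clamped local-position Z, appended with two constant ones, supplies the function's TechArtMask input.
# Ray-traced shadow casting is disabled.
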Material /Game/Global/Materials/MaterialGym/Materials/OverlayTransparent/M_OverlayTrans_Generic
graph {
 properties {
  bCastRayTracedShadows=false
  bUseMaterialAttributes=true
  bUsedWithMeshDeformer=true
  bUsedWithSkeletalMesh=true
  blendMode=BLEND_Translucent
  pixelDepthOffsetMode=PDOM_Legacy
  shadingModel=MSM_Unlit
  usedShadingModels=Unlit
 }
 nodes [
  n1 = SetMaterialAttributes!(
   MaterialAttributes
   Base Color=VectorParameter() {group=Showcase, parameterName=BaseColor_RE, defaultValue=<rgba 0.0223 0.1069 0.2188 1>}.RGB
   Metallic=Constant()
   Roughness=ScalarParameter() {defaultValue=0.15, group=Showcase, parameterName=Roughness_RE}
   Emissive Color=VectorParameter() {group=Showcase, parameterName=Emissive_RE, defaultValue=<rgba 0.0212 0.0704 0.1354 1>}.RGB
   Normal=TextureSample() {samplerType=SAMPLERTYPE_LinearColor, texture=/Game/Global/Textures/Default/Noise/TilingNoise06_N_uncompressed}.RGB
  ) {attributeSetTypes [
    #1
    #2
    #3
    #4
    #5
   ]}
  n2 = ComponentMask(None=MaterialFunctionCall() {materialFunction=&LocalPosition}.Local Position) {b=true}
  MaterialFunctionCall!(Material=@n1, TechArtMask=AppendVector(A=Clamp(None=@n2), B=Constant2Vector() {g=1, r=1})) {materialFunction=&/MF_OverlayTrans_Generic}
 ]
}
