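# Material function that layers a fake-transparency overlay onto input material attributes, faking translucency through an opacity mask.
# A tech-art channel mask, two-sided sign and extra mask texture shape the opacity, with separate smoothing for mobile quality tiers.
# Base color gets sky-light tint and darkening.
# An optional inner-color override lerps each attribute, since two-sided sign is unavailable in vertex shaders, and can push vertices toward the camera by Fresnel.
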MaterialFunction /Game/Global/Materials/MaterialGym/Materials/OverlayTransparent/MF_SilverBellFakeTrans_Generic
graph {nodes [
  n1 = GetMaterialAttributes(None=NamedRerouteUsage()) {attributeGetTypes [#1, #2, #3]}
  n2 = NamedRerouteUsage()
  n3 = LinearInterpolate(
   A=@n1.BaseColor
   B=Multiply(A=Multiply(A=@n1.BaseColor, B=@n2), B=ScalarParameter() {defaultValue=1, group=NewFakeTransOverlay, parameterName=SkinBrightness})
   Alpha=NamedRerouteUsage()
  )
  n4 = ScalarParameter() {defaultValue=0.5, group=NewFakeTransOverlay, parameterName=DarkenBaseColor}
  n5 = SetMaterialAttributes!(
   MaterialAttributes=@n1.MaterialAttributes
   OpacityMask=NamedRerouteUsage()
   BaseColor=LinearInterpolate(A=@n3, B=Multiply(A=@n3, B=@n4), Alpha=NamedRerouteUsage())
   EmissiveColor=NamedRerouteUsage()
  ) {attributeSetTypes [#4, #1, #2]}
  n6 = GetMaterialAttributes(None=@n5) {
   desc="Two sided sign is not avavilable in vertex shader, So blend material is not working.\r\nHave to lerp each parameters"
   attributeGetTypes [
    #5
    #1
    #2
    #6
    #7
    #3
   ]
  }
  n7 = NamedRerouteUsage()
  n8 = VectorParameter() {group=NewFakeTransOverlay, parameterName=EmissiveColorOverride, defaultValue=<rgba 0.0452 0.0762 0.1095 1>}
  n9 = Constant() {r=0.2}
  n10 = Normalize(VectorInput=Subtract(A=CameraPositionWS(), B=WorldPosition().XYZ))
  n11 = DotProduct(A=VertexNormalWS(), B=@n10)
  n12 = ScalarParameter() {defaultValue=0.7, group=NewFakeTransOverlay, parameterName=WPOFrenellPower}
  n13 = LinearInterpolate(
   A=Constant()
   B=Multiply(A=@n10, B=ScalarParameter() {defaultValue=1, group=NewFakeTransOverlay, parameterName=PushtoCamera})
   Alpha=Saturate(None=Power(Base=@n11, Exp=@n12))
  )
  n14 = NamedRerouteUsage()
  n15 = SetMaterialAttributes!(
   MaterialAttributes=@n6.MaterialAttributes
   BaseColor=LinearInterpolate(
    A=@n6.BaseColor
    B=VectorParameter() {group=NewFakeTransOverlay, parameterName=BaseColorOverride, defaultValue=<rgba 0.1779 0.305 0.4452 1>}.RGB
    Alpha=NamedRerouteUsage()
   )
   Metallic=LinearInterpolate(A=@n6.Metallic, B=Constant(), Alpha=@n7)
   Roughness=LinearInterpolate(A=@n6.Roughness, B=Constant() {r=1}, Alpha=@n7)
   EmissiveColor=LinearInterpolate(A=@n6.EmissiveColor, B=@n8.RGB, Alpha=NamedRerouteUsage())
   Specular=LinearInterpolate(A=@n6.Specular, B=Constant(), Alpha=NamedRerouteUsage())
   Normal=LinearInterpolate(
    A=@n6.Normal
    B=MaterialFunctionCall(Normal=@n6.Normal, Flatness=@n9) {materialFunction=&FlattenNormal}.Result
    Alpha=NamedRerouteUsage()
   )
   WorldPositionOffset=StaticSwitchParameter(True=LinearInterpolate(B=@n13, Alpha=@n14) {constB=0}, False=Constant()) {desc=Optional Fix, group=NewFakeTransOverlay, parameterName=WPOTowardsCamera}
  ) {attributeSetTypes [
    #1
    #6
    #7
    #2
    #3
    #5
    #8
   ]}
  FunctionOutput(None=StaticSwitchParameter(True=@n15, False=@n5) {group=NewFakeTransOverlay, parameterName=OverrideInnerColor})
  NamedRerouteDeclaration(None=FunctionInput() {inputName=Material, inputType=FunctionInput_MaterialAttributes}) {name=MA, nodeColor=<rgba 0.7344 1 0 1>}
  n16 = ChannelMaskParameter(None=FunctionInput() {inputName=TechArtMask, sortPriority=1}) {group=NewFakeTransOverlay, parameterName=FakeTransChannel}
  NamedRerouteDeclaration(None=StaticSwitchParameter(True=OneMinus(None=@n16.R), False=@n16.R) {group=NewFakeTransOverlay, parameterName=FlipTechArtMask}) {name=TechArtMask, nodeColor=<rgba 0.5703 1 0 1>}
  NamedRerouteDeclaration(None=@n1.MaterialAttributes) {name=MA_IN, nodeColor=<rgba 1 0.7969 0 1>}
  n17 = TextureSampleParameter2D() {
   group=NewFakeTransOverlay
   parameterName=ExtraOpacityOverrideMask
   samplerType=SAMPLERTYPE_Masks
   texture=&T_WhiteMask
  }
  NamedRerouteDeclaration(None=@n17.G) {name=ColorOverrideMask, nodeColor=<rgba 0 0.1328 1 1>}
  n18 = Step(X=TwoSidedSign())
  NamedRerouteDeclaration(None=Multiply(A=LinearInterpolate(Alpha=@n18) {constA=1, constB=0}, B=NamedRerouteUsage())) {name=BackColorMask, nodeColor=<rgba 1 0 0.375 1>}
  n19 = SkyLightEnvMapSample(
   Direction=Multiply(A=CameraVectorWS()) {constB=-1}
   Roughness=ScalarParameter() {defaultValue=0.22, group=NewFakeTransOverlay, parameterName=SkyTintRoughness}
  )
  NamedRerouteDeclaration(None=Clamp(
    None=Desaturation(None=@n19)
    Min=ScalarParameter() {group=NewFakeTransOverlay, parameterName=SkyTintMin}
    Max=ScalarParameter() {defaultValue=1, group=NewFakeTransOverlay, parameterName=SkyTintMax}
   )) {name=SkyBoxTint, nodeColor=<rgba 1 0 0.1641 1>}
  n20 = SmoothStep(
   Min=ScalarParameter() {group=NewFakeTransOverlay, parameterName=TechArtMaskSmoothLow}
   Max=ScalarParameter() {defaultValue=1, group=NewFakeTransOverlay, parameterName=TechArtMaskSmoothHigh}
   Value=NamedRerouteUsage()
  )
  n21 = LinearInterpolate(Alpha=Step(X=TwoSidedSign())) {constA=1, constB=0}
  n22 = Multiply(
   A=StaticSwitchParameter(True=LinearInterpolate(Alpha=@n20) {constA=1, constB=0}, False=LinearInterpolate(B=@n21, Alpha=@n20) {constA=1}) {group=NewFakeTransOverlay, parameterName=UseTechArtMaskAsFullOpacity}
   B=@n17.R
  )
  n23 = SmoothStep(Max=ScalarParameter() {defaultValue=1, group=NewFakeTransOverlay, parameterName=OpacitySmoothMax}, Value=@n22) {constMax=0.7}
  n24 = SmoothStep(
   Min=ScalarParameter() {defaultValue=0.7, group=NewFakeTransOverlay, parameterName=OpacitySmoothMinLowEnd}
   Max=ScalarParameter() {defaultValue=0.8, group=NewFakeTransOverlay, parameterName=OpacitySmoothMaxLowEnd}
   Value=@n22
  ) {constMax=0.7, constMin=0.5}
  NamedRerouteDeclaration(None=MaterialFunctionCall!(
    Low _ PC Low and Med=@n23
    "High _ XB1, PS4"=@n23
    "Epic _ PC High, XSX, PS5"=@n23
    --------
    "Mobile Low _ PC Perf, Android"=@n24
    Mobile Med _ Switch=@n24
   ) {desc=DitherAA_Adjustment, materialFunction=/Game/Global/MaterialFunctions/TechArt/MF_Fortnite_QualitySwitch_Scalar}.Result) {name=OpacityMask_OUT, nodeColor=<rgba 0 0.8594 1 1>}
  NamedRerouteDeclaration(None=@n1.EmissiveColor) {name=EM_IN, nodeColor=<rgba 0.2422 1 0 1>}
  n25 = NamedRerouteUsage()
  n26 = Multiply(A=Multiply(A=@n25, B=OneMinus(None=@n25))) {constB=4}
  NamedRerouteDeclaration(None=StaticSwitchParameter(True=Saturate(None=@n26), False=Constant()) {group=NewFakeTransOverlay, parameterName=OverrideBaseColor}) {name=OverrideBaseColorMask, nodeColor=<rgba 0.6875 1 0 1>}
  NamedRerouteDeclaration(None=@n17.B) {name=WPOOverrideMask, nodeColor=<rgba 0 0.6016 1 1>}
  NamedRerouteDeclaration(None=@n1.Specular) {name=Spec_IN, nodeColor=<rgba 1 0 0.9609 1>}
 ]}
