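# Material gym showcase of a two-sided, masked, dithered material built around the MF_SilverBellFakeTrans_Generic fake-transparency function.
# Clamped local-position Z blends base color from white to a tint parameter, with non-metallic roughness and emissive parameters set through material attributes.
# The same gradient, appended with two ones, feeds the function's TechArtMask input alongside those attributes.
# Usage flags cover Nanite, skeletal meshes, cloth, mesh deformers and Niagara mesh particles.
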Material /Game/Global/Materials/MaterialGym/Materials/OverlayTransparent/M_TwosidedTransparent_Generic
graph {
 properties {
  bUseMaterialAttributes=true
  bUsedWithClothing=true
  bUsedWithMeshDeformer=true
  bUsedWithNanite=true
  bUsedWithNiagaraMeshParticles=true
  bUsedWithSkeletalMesh=true
  blendMode=BLEND_Masked
  ditherOpacityMask=true
  pixelDepthOffsetMode=PDOM_Legacy
  shadingModel=MSM_FromMaterialExpression
  twoSided=true
  usedShadingModels=Default Lit
 }
 nodes [
  n1 = Clamp(None=ComponentMask(None=MaterialFunctionCall() {materialFunction=&LocalPosition}.Local Position) {b=true})
  n2 = SetMaterialAttributes!(
   MaterialAttributes
   Base Color=LinearInterpolate(
    A=Constant() {r=1}
    B=VectorParameter() {group=Showcase, parameterName=BaseColor_RE, defaultValue=<rgba 0.0223 0.1069 0.2188 1>}.RGB
    Alpha=@n1
   )
   Metallic=Constant()
   Roughness=ScalarParameter() {defaultValue=0.15, group=Showcase, parameterName=Roughness_RE}
   Emissive Color=VectorParameter() {group=Showcase, parameterName=Emissive_RE, defaultValue=<rgba 0.0212 0.0704 0.1354 1>}.RGB
  ) {attributeSetTypes [
    #1
    #2
    #3
    #4
   ]}
  MaterialFunctionCall!(Material=@n2, TechArtMask=AppendVector(A=@n1, B=Constant2Vector() {g=1, r=1})) {materialFunction=&/MF_SilverBellFakeTrans_Generic}
 ]
}
